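; Shooting gallery target that knocks down when its mesh is shot, then pops back up, and never dies.
; Server damage on the target mesh passes a DoOnce into KnockDown, which plays FX, drops aim assist and rotates the mesh by timeline.
; A random-delay timer then runs PopUp, which reverses the state, re-registers aim assist and calls DamageReset on the DoOnce.
; Weak-spot FX exist, but damage always passes false for that flag.
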(event EventBeginPlay
  ;[1 1 1] EVENT -> Begin Play - Enable aim assist
  (AimAssist|RegisterAimAssistTarget self (Variables|Default|GetAimAssistCollision)))

(event EventOnDamageServer
  (params (Damage :type Float%20%28single-precision%29) (DamageTags :type Gameplay%20Tag%20Container%20Structure%20%28by%20ref%29) (Momentum :type Vector :default "0, 0, 0") (HitInfo :type Hit%20Result%20Structure%20%28by%20ref%29) (InstigatedBy :type Controller%20Object%20Reference) (DamageCauser :type Actor%20Object%20Reference) (EffectContext :type Gameplay%20Effect%20Context%20Handle%20Structure))
  (bind (_bblockinghit _binitialoverlap _time _distance _location _impactpoint _normal _impactnormal _physmat _hitactor _hitcomponent) (Collision|BreakHitResult HitInfo))
  ;[1 1 1] EVENT -> DamageOnServer - Knock down the target on any damage
  (if (Utilities|Equal(Object) _hitcomponent (Variables|Default|GetSM_Target))
    (Utilities|FlowControl|DoOnce false
      (:Completed
        (CallFunction|KnockDown false HitInfo)))))

(event Custom|DamageReset
  ;[1 1 1] EVENT -> DamageOnServer - Knock down the target on any damage
  (Utilities|FlowControl|DoOnce false
    (:Completed
      (CallFunction|KnockDown false (stale EventOnDamageServer HitInfo)))))

(event Custom|KnockDown
  (params (HitWeakSpot :type Boolean) (HitResult :type Hit%20Result%20Structure))
  ;[1 1 1] Custom Event -> Knock down the target and start the Reset timer
  (if (not (Variables|Default|GetisDown))
    (Variables|Default|SetisDown true)
    (ShootingTarget|KnockDownFX HitWeakSpot HitResult)
    (AimAssist|UnregisterAimAssistTarget self)
    (bind (_direction _rotation) (Timeline|KnockDown_Timeline PlayFromStart)
      (:Update
        (Transformation|SetRelativeRotation (Variables|Default|GetSM_Target) (Math|Rotator|MakeRotator _rotation) false false))
      (:Finished
        (CallFunction|Reset)))))

(event Custom|Reset
  ;[1 1 1] Custom Event -> Reset Timer
  (bind _returnvalue (Utilities|Time|SetTimerbyEvent (delegate Custom|PopUp) (Utilities|ClampMinFloat (Math|Random|RandomFloatinRange (Variables|Target|GetPopUpDelayMin) (Variables|Target|GetPopUpDelayMax)) 0.01)))
  (Variables|Default|SetResetTimer _returnvalue))

(event Custom|PopUp
  ;[1 1 1] PopUp the Target
  (if (Variables|Default|GetisDown)
    (Utilities|Time|ClearandInvalidateTimerbyHandle (Variables|Default|GetResetTimer))
    (Variables|Default|SetisDown false)
    (ShootingTarget|PopUpFX)
    (bind (_direction _rotation) (Timeline|PopUp_Timeline PlayFromStart)
      (:Update
        (Transformation|SetRelativeRotation (Variables|Default|GetSM_Target) (Math|Rotator|MakeRotator _rotation) false false))
      (:Finished
        (AimAssist|RegisterAimAssistTarget self (Variables|Default|GetAimAssistCollision))
        (CallFunction|DamageReset)))))

(fn KnockDownFX
  (params (HitWeakSpot :type Boolean) (HitResult :type Hit%20Result%20Structure))
  (bind (_bblockinghit _binitialoverlap _time _distance _location _impactpoint _normal _impactnormal) (Collision|BreakHitResult HitResult))
  (Audio|PlaySoundatLocation (Variables|Audio|GetSFX_TargetHit) _location)
  (if HitWeakSpot
    (Audio|PlaySoundatLocation (Variables|Audio|GetSFX_TargetWeakSpotHit) _location)
    (Effects|SpawnEmitteratLocation (Variables|FX|GetFX_WeakSpotHit) _location "-90.000000, 0,0.000000")
    (else
      (Effects|SpawnEmitteratLocation (Variables|FX|GetFX_KnockDown) _location (Math|Conversions|RotationFromXVector _impactnormal)))))

(fn PopUpFX
  (bind _aimassistcollision (Variables|Default|GetAimAssistCollision))
  (bind _returnvalue (Transformation|GetWorldLocation _aimassistcollision))
  (Effects|SpawnEmitteratLocation (Variables|FX|GetFX_PopUp) _returnvalue)
  (Audio|PlaySoundatLocation (Variables|Audio|GetSFX_PopUp) _returnvalue "0, 0, 0" 1.5))

(fn ShouldDie
  (params (Damage :type Float%20%28single-precision%29) (EventInstigator :type Controller%20Object%20Reference) (DamageCauser :type Actor%20Object%20Reference) (bIsTest :type Boolean) (EffectContext :type Gameplay%20Effect%20Context%20Handle%20Structure%20%28by%20ref%29))
  (return false))

(fn ConstructionScript)
